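# Additive unlit Niagara mesh-particle material for the Diamond Flare aura, rendering camera-offset glowing sprites via world position offset.
# Time and vertex color drive a ripple that, with panning radial noise, modulates sprite scale and a three-color ramp.
# Particle color and an edge-to-center sphere mask tint the glow; particle alpha, mask and camera-distance fade combine for opacity.
# The mask's hardness reuses its radius parameter, and hue-shift and rainbow branches appear unused.
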Material /SparksCosmetics/Items/Auras/Assets/NS_Aura_DiamondFlare/M_Aura_DiamondFlare_MeshSprite
graph {
 properties {
  bUsedWithNiagaraMeshParticles=true
  blendMode=BLEND_Additive
  pixelDepthOffsetMode=PDOM_Legacy
  shadingModel=MSM_Unlit
  usedShadingModels=Unlit
 }
 nodes [
  n1 = DynamicParameter() {defaultValue=<rgba 3.5 -30 0.5 0.5>, paramNames [SpriteSize, Camera Offset, Sphere Mask Radius, Sphere Mask Hardness]}
  n2 = Time()
  n3 = ScalarParameter() {defaultValue=1.5, parameterName=Ripple Speed Mult}
  n4 = VertexColor()
  n5 = Frac(None=Add(A=Multiply(A=@n2, B=@n3), B=Multiply(A=@n4) {constB=2}))
  n6 = Add(A=Multiply(A=Sine(None=@n5)) {constB=0.5}) {constB=0.5}
  n7 = MaterialFunctionCall(Vector or UVs=Add(A=TextureCoordinate() {coordinateIndex=1}) {constB=-0.5}) {materialFunction=&VectorToRadialValue}
  n8 = TextureSample(UVs=Panner(Coordinate=@n7.Radial Coordinates) {speedY=-0.5}) {samplerType=SAMPLERTYPE_LinearGrayscale, texture=/Game/Effects/Fort_Effects/Emotes/UltraLight/Texture/T_DirectionalNoise_Greyscale}
  n9 = Multiply(A=@n8.B) {constB=0.3}
  n10 = AppendVector(A=Multiply(A=@n6, B=Add(A=@n9) {constB=0.7}), B=Constant() {r=1})
  Add(
   A=MaterialFunctionCall!(Center Location=WorldPosition().XYZ, 0-1 UVs, XY scale=Multiply(A=@n10, B=@n1.SpriteSize), OverridePivot, "Normals (Optional)") {materialFunction=&Sprite}.World position offset
   B=MaterialFunctionCall(Offset Amount=@n1.Camera Offset, World Position=WorldPosition().XYZ) {materialFunction=/Engine/Functions/Engine_MaterialFunctions02/WorldPositionOffset/CameraOffset}.Clamped Camera Offset
  )
  Constant3Vector() {constant=<rgba 1.9232 2 0.233 1>}
  Constant3Vector() {constant=<rgba 0.1031 0.0426 0.0061 0>}
  Constant3Vector() {constant=<rgba 0.3061 0.209 0.0804 0>}
  MaterialExpressionTextureCoordinate_0 = TextureCoordinate() {coordinateIndex=1}
  Convert(None=@MaterialExpressionTextureCoordinate_0) {
   nodeName=Break Float2
   convertInputs [{
     type=Vector2
     expressionInput {
      expression=&self:MaterialExpressionTextureCoordinate_0
      inputName=None
      mask=0
      maskA=0
      maskB=0
      maskG=0
      maskR=0
      outputIndex=0
     }
    }]
   convertOutputs [{type=Scalar}, {type=Scalar}]
  }
  n11 = MaterialFunctionCall(
   A=Constant3Vector() {constant=<rgba 0.1878 0.0289 0.0062 1>}
   B=Constant3Vector() {constant=<rgba 0.3061 0.1871 0.0293 0>}
   C=Constant3Vector() {constant=<rgba 2 1.8285 1.268 1>}
   Alpha=Multiply(A=@n6, B=@n8.B)
  ) {materialFunction=&Lerp_3Color}
  n12 = DynamicParameter() {defaultValue=<rgba 3.5 -30 0.5 0.5>, paramNames [SpriteSize, Camera Offset, Sphere Mask Radius, Sphere Mask Hardness]}
  n13 = SphereMask(A=TextureCoordinate(), B=Constant() {r=0.5}, Radius=@n12.Sphere Mask Radius, Hardness=@n12.Sphere Mask Radius)
  n14 = LinearInterpolate(
   A=VectorParameter() {parameterName=Edge Color, defaultValue=<rgba 0.6896 0.6771 1 1>}.RGB
   B=VectorParameter() {parameterName=Center Color, defaultValue=<rgba 1 0.9335 0.5823 1>}.RGB
   Alpha=@n13
  )
  n15 = ParticleColor()
  MaterialFunctionCall(Emissive=Multiply(A=@n11.Result, B=Multiply(A=@n15.RGB, B=@n14))) {materialFunction=&MF_FX_EmissiveAdjustments}
  n16 = Multiply(A=VertexColor()) {constB=10}
  n17 = Multiply(A=Ceil(None=Add(A=@n16) {constB=0.1})) {constB=0.1}
  n18 = Constant() {r=2}
  n19 = Sine(None=Multiply(A=@n17, B=@n18) {constB=0.5})
  n20 = Sine(None=Add(A=Multiply(A=@n17, B=@n18) {constB=0.5}) {constB=-0.66})
  n21 = Sine(None=Add(A=Multiply(A=@n17, B=@n18) {constB=0.5}) {constB=0.66})
  MaterialFunctionCall(
   R=Add(A=Multiply(A=@n19) {constB=0.5}) {constB=0.58}
   G=Add(A=Multiply(A=@n20) {constB=0.5}) {constB=0.58}
   B=Add(A=Multiply(A=@n21) {constB=0.5}) {constB=0.58}
  ) {materialFunction=&AppendMany}
  n22 = TextureSample(UVs=Panner(Coordinate=@n7.Radial Coordinates) {speedY=-0.3}) {samplerType=SAMPLERTYPE_LinearGrayscale, texture=&T_DirectionalNoise_Greyscale}
  MaterialFunctionCall(Hue Shift Percentage=@n22.G, Texture=@n11.Result) {materialFunction=&HueShift}
  n23 = OneMinus(None=VertexColor())
  Multiply(A=@n11.Result, B=Add(B=Multiply(A=@n23) {constB=2}) {constA=0.5})
  Multiply(
   A=Multiply(A=@n15.A, B=@n13)
   B=MaterialFunctionCall!(Distance=ScalarParameter() {defaultValue=300, parameterName=Camera Fade Distance}) {materialFunction=/Game/Effects/Fort_Effects/Materials/_Functions/MF_CameraDistanceFade}.Result
  )
  Multiply(A=Add(A=@n22.G) {constB=-0.2}) {constB=0.5}
 ]
}
